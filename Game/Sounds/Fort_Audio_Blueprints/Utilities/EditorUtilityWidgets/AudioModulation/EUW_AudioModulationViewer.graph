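# Editor utility widget that shows live audio modulation values for chosen modulators and the buses of a chosen Control Bus Mix.
# Property-view picks call TrackBus, which skips duplicates, creates a modulation destination and adds an AudioModulationBus_Widget entry coloured by patch or control bus type.
# Tick updates each entry's bar and text from GetModulatorValue, using the first PIE world, else the editor world.
# Reset clears tracked state and entries.
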EditorUtilityWidgetBlueprint /Game/Sounds/Fort_Audio_Blueprints/Utilities/EditorUtilityWidgets/AudioModulation/EUW_AudioModulationViewer
graphs [
 EventGraph {nodes [
   "On Property Changed (ControlBusField)"!() -> (OutputDelegate, then, PropertyName:Name=None) {is=title K2Node_ComponentBoundEvent}
   > n1 = Get!(execute:Exec) -> (then:Exec, else:Exec, ModulatorBase:Sound Modulator Base Object Reference)
   > TrackBus!(execute, self, ModulatorBase:Sound Modulator Base Object Reference=@n1) -> (then)
   "On Pressed (ResetButton)"() {is=title K2Node_ComponentBoundEvent}
   > Clear(TargetArray:Array of Sound Modulator Base Object References=Get AddedModulatorBases() -> (AddedModulatorBases:Array of Sound Modulator Base Object References))
   > Clear(TargetArray:Array of Audio Modulation Destination Object References=Get ModulatorDestinations() -> (ModulatorDestinations:Array of Audio Modulation Destination Object References))
   > Reset!(
    execute
    self:Dynamic Entry Box Object Reference=Get ControlBusses() -> (ControlBusses:Dynamic Entry Box Object Reference)
    bDeleteWidgets:bool=true
   ) -> (then)
   > Set ModulatorBase(ModulatorBase:Sound Modulator Base Object Reference) -> (Output_Get:Sound Modulator Base Object Reference)
   > Set ControlBusMix(ControlBusMix:Sound Control Bus Mix Object Reference) -> (Output_Get:Sound Control Bus Mix Object Reference)
   > SetObject(self=Get ControlBusField() -> (ControlBusField:Single Property View Object Reference), NewObject=Self-Reference())
   > SetObject(self=Get ControlBusMixField() -> (ControlBusMixField:Single Property View Object Reference), NewObject=Self-Reference())
   "On Property Changed (ControlBusMixField)"!() -> (OutputDelegate, then, PropertyName:Name=None) {is=title K2Node_ComponentBoundEvent}
   > n2 = Get!(execute) -> (then, else, ControlBusMix:Sound Control Bus Mix Object Reference)
   > n3 = For Each Loop(Array:Array of Sound Control Bus Mix Stage Structures=Get MixStages!(self:Sound Control Bus Mix Object Reference=@n2) -> (MixStages:Array of Sound Control Bus Mix Stage Structures)) -> (Array Element:Sound Control Bus Mix Stage Structure)
   > TrackBus(ModulatorBase=Break Sound Control Bus Mix Stage!(SoundControlBusMixStage:"Sound Control Bus Mix Stage Structure (by ref)"=@n3) -> (Bus:Sound Control Bus Object Reference, Value_TargetValue:float=~, Value_AttackTime:float=~, Value_ReleaseTime:float=~) {is=title K2Node_BreakStruct})
   Event PreConstruct()
   n4 = Get ControlBusField()
   > SetObject(self=@n4, NewObject=Self-Reference())
   > SetPropertyName!(execute, self:Single Property View Object Reference=@n4, NewPropertyName:Name=ModulatorBase) -> (then)
   n5 = Get ControlBusMixField()
   > SetObject(self=@n5, NewObject=Self-Reference())
   > SetPropertyName(self=@n5, NewPropertyName=ControlBusMix)
   Event Tick()
   > n6 = For Each Loop(Array:Array of Sound Modulator Base Object References=Get AddedModulatorBases()) -> (Array Element:Sound Modulator Base Object Reference)
   n7 = Get ControlBusses()
   > n8 = Cast To AudioModulationBus_Widget!(
    execute
    Object="Get (a copy)"(
     Array:Array of User Widget Object References=GetAllEntries!(self:Dynamic Entry Box Base Object Reference=@n7) -> (ReturnValue:Array of User Widget Object References)
     Dimension 1=@n6.Array Index
    ) -> (Output:User Widget Object Reference)
   ) -> (then, CastFailed, AsAudio Modulation Bus Widget:Audio Modulation Bus Widget Object Reference) {is=title K2Node_DynamicCast}
   n9 = Get Modulator Value!(
    WorldContextObject:Object Reference=GetCurrentWorld!() -> (ReturnValue:World Object Reference)
    Modulator:Sound Modulator Base Object Reference=@n6
   ) -> (ReturnValue:float) {is=title K2Node_CallFunction}
   > SetPercent!(
    execute
    self:Progress Bar Object Reference=Get ProgressBar_0!(self:Audio Modulation Bus Widget Object Reference=@n8) -> (ProgressBar_0:Progress Bar Object Reference)
    InPercent:float=@n9
   ) -> (then)
   > Set Text(
    Text="To Text (Float)"(Value=@n9, MinimumIntegralDigits=1, MaximumIntegralDigits=1, MinimumFractionalDigits=2, MaximumFractionalDigits=2)
    self=Get Value!(self:Audio Modulation Bus Widget Object Reference=@n8) -> (Value:Text Block Object Reference)
   )
  ]}
 GetCurrentWorld {nodes [
   GetCurrentWorld() {is=title K2Node_FunctionEntry}
   > n1 = GetPIEWorlds!(execute, bIncludeDedicatedServer:bool=false) -> (then, ReturnValue:Array of World Object References)
   > n2 = Branch(Condition=Is Valid Index!(TargetArray:Array of World Object References=@n1, IndexToTest:int) -> (ReturnValue:bool))
   > Return Node!(execute, ReturnValue:World Object Reference="Get (a copy)"(Array:Array of World Object References=@n1) -> (Output:World Object Reference))
   n3 = GetEditorWorld(execute=@n2.else, self=Get UnrealEditorSubsystem())
   > Is Valid(InputObject=@n3)
   > Return Node(ReturnValue=@n3)
  ]}
 TrackBus {nodes [
   n1 = TrackBus!() -> (then, ModulatorBase:Sound Modulator Base Object Reference) {is=title K2Node_FunctionEntry}
   > n2 = Branch(Condition=Contains Item(
     TargetArray:Array of Sound Modulator Base Object References=Get AddedModulatorBases()
     ItemToFind:"Sound Modulator Base Object Reference (by ref)"=@n1
    ))
   Add(
    execute=@n2.else
    TargetArray:Array of Sound Modulator Base Object References=Get AddedModulatorBases()
    NewItem:"Sound Modulator Base Object Reference (by ref)"=@n1
   ) {is=title K2Node_CallArrayFunction}
   > n3 = CreateModulationDestination!(
    execute
    WorldContextObject:Object Reference=GetCurrentWorld() {is=title K2Node_CallFunction}
    Name:Name=String To Name(InString=GetDisplayName(Object=@n1))
    Modulator:Sound Modulator Base Object Reference=@n1
   ) -> (then, ReturnValue:Audio Modulation Destination Object Reference)
   > Add(
    TargetArray:Array of Audio Modulation Destination Object References=Get ModulatorDestinations()
    NewItem:"Audio Modulation Destination Object Reference (by ref)"=@n3
   )
   > n4 = Create Entry!(execute, self:Dynamic Entry Box Object Reference=Get ControlBusses()) -> (then, ReturnValue:User Widget Object Reference)
   > n5 = Cast To AudioModulationBus_Widget(Object=@n4)
   > n6 = Cast To SoundModulationPatch!(execute, Object=@n1) -> (then, CastFailed, AsSound Modulation Patch:Sound Modulation Patch Object Reference) {is=title K2Node_DynamicCast}
   n7 = Get ProgressBar_0(self=@n5)
   > n8 = SetFillColorAndOpacity!(execute, self:Progress Bar Object Reference=@n7, InColor:Linear Color Structure=(R=1,G=0.6038,B=1,A=1)) -> (then)
   Set Text(
    execute=[@n8.then, @n9.then]
    Text="To Text (String)"(InString=GetObjectName(Object=@n1))
    self=Get UEFN_TextBlock_C_0!(self:Audio Modulation Bus Widget Object Reference=@n5) -> (UEFN_TextBlock_C_0:Text Block Object Reference)
   )
   Cast To SoundControlBus!(execute=@n6.CastFailed, Object=@n1) -> (then, CastFailed, AsSound Control Bus:Sound Control Bus Object Reference) {is=title K2Node_DynamicCast}
   > n9 = SetFillColorAndOpacity(self=@n7, InColor=(R=1,G=0.0331,B=0.3185,A=1))
  ]}
]
variables [
 AddedModulatorBases {type="array<SoundModulatorBase>", default []}
 AudioBus {type=AudioBus, default=null}
 ControlBusMix {type=SoundControlBusMix, default=null}
 ModulatorBase {type=SoundModulatorBase, default=null}
 ModulatorDestinations {type="array<AudioModulationDestination>", default []}
]
widgets {
 info {rootWidgetClass=&CanvasPanel, widgetCount=8}
 widgets [
  CanvasPanel_0 {widgetClassPath=&CanvasPanel, widget=CanvasPanel_0}
  ScaleBox_26 {
   widgetClassPath=&ScaleBox
   parent=CanvasPanel_0
   widget=ScaleBox_26
   slot {
    class=CanvasPanelSlot
    properties {layoutData {alignment=<0 0>, anchors {maximum=<0 0>, minimum=<0 0>}, offsets=<ltrb 0 140 499.8998 323.7938>}}
   }
  }
  ControlBusField {
   widgetClassPath=&SinglePropertyView
   parent=CanvasPanel_0
   widget=ControlBusField
   slot {class=CanvasPanelSlot, properties {layoutData {alignment=<0 0>, anchors {maximum=<0 0>, minimum=<0 0>}, offsets=<ltrb 0 70 380 60>}}}
  }
  ResetButton {
   widgetClassPath=&Button
   parent=CanvasPanel_0
   widget=ResetButton
   properties {widgetStyle {
     clickedSlateSound {resourceObject=null}
     disabled {
      bIsDynamicallyLoaded=false
      drawAs=RoundedBox
      imageType=NoImage
      mirroring=NoMirror
      resourceName=None
      resourceObject=null
      tiling=NoTile
      imageSize=<0 0>
      margin=<ltrb 0 0 0 0>
      outlineSettings {
       bUseBrushTransparency=false
       roundingType=FixedRadius
       width=1
       color {colorUseRule=UseColor_ColorTable, specifiedColor=<rgba 0 0 0 0>}
       cornerRadii=<4 4 4 4>
      }
      tintColor {colorUseRule=UseColor_Specified, specifiedColor=<rgba 0.0395 0.0395 0.0395 1>}
      uVRegion {bIsValid=false, max=<0 0>, min=<0 0>}
     }
     disabledForeground {colorUseRule=UseColor_ColorTable, specifiedColor=<rgba 0 0 0 0>}
     hovered {
      bIsDynamicallyLoaded=false
      drawAs=RoundedBox
      imageType=NoImage
      mirroring=NoMirror
      resourceName=None
      resourceObject=null
      tiling=NoTile
      imageSize=<32 32>
      margin=<ltrb 0 0 0 0>
      outlineSettings {
       bUseBrushTransparency=false
       roundingType=FixedRadius
       width=1
       color {colorUseRule=UseColor_ColorTable, specifiedColor=<rgba 0 0 0 0>}
       cornerRadii=<4 4 4 4>
      }
      tintColor {colorUseRule=UseColor_Specified, specifiedColor=<rgba 0.0953 0.0953 0.0953 1>}
      uVRegion {bIsValid=false, max=<0 0>, min=<0 0>}
     }
     hoveredForeground {colorUseRule=UseColor_ColorTable, specifiedColor=<rgba 0 0 0 0>}
     hoveredSlateSound {resourceObject=null}
     normal {
      bIsDynamicallyLoaded=false
      drawAs=RoundedBox
      imageType=NoImage
      mirroring=NoMirror
      resourceName=None
      resourceObject=null
      tiling=NoTile
      imageSize=<32 32>
      margin=<ltrb 0 0 0 0>
      outlineSettings {
       bUseBrushTransparency=false
       roundingType=FixedRadius
       width=1
       color {colorUseRule=UseColor_ColorTable, specifiedColor=<rgba 0 0 0 0>}
       cornerRadii=<4 4 4 4>
      }
      tintColor {colorUseRule=UseColor_Specified, specifiedColor=<rgba 0.0395 0.0395 0.0395 1>}
      uVRegion {bIsValid=false, max=<0 0>, min=<0 0>}
     }
     normalForeground {colorUseRule=UseColor_ColorTable, specifiedColor=<rgba 0 0 0 0>}
     normalPadding=<ltrb 12 1.5 12 1.5>
     pressed {
      bIsDynamicallyLoaded=false
      drawAs=RoundedBox
      imageType=NoImage
      mirroring=NoMirror
      resourceName=None
      resourceObject=null
      tiling=NoTile
      imageSize=<32 32>
      margin=<ltrb 0 0 0 0>
      outlineSettings {
       bUseBrushTransparency=false
       roundingType=FixedRadius
       width=1
       color {colorUseRule=UseColor_ColorTable, specifiedColor=<rgba 0 0 0 0>}
       cornerRadii=<4 4 4 4>
      }
      tintColor {colorUseRule=UseColor_Specified, specifiedColor=<rgba 0.0284 0.0284 0.0284 1>}
      uVRegion {bIsValid=false, max=<0 0>, min=<0 0>}
     }
     pressedForeground {colorUseRule=UseColor_ColorTable, specifiedColor=<rgba 0 0 0 0>}
     pressedPadding=<ltrb 12 2.5 12 0.5>
     pressedSlateSound {resourceObject=null}
    }}
   slot {class=CanvasPanelSlot, properties {layoutData {alignment=<0 0>, anchors {maximum=<0 0>, minimum=<0 0>}, offsets=<ltrb 404 12 100 40>}}}
  }
  CommonTextBlock_0 {
   widgetClassPath=&CommonTextBlock
   parent=ResetButton
   widget=CommonTextBlock_0
   properties {
    style=&TextStyle-Base_C
    text=Reset
    font {
     bForceMonospaced=false
     bMaterialIsStencil=false
     fontMaterial=null
     fontObject=&NotoSans
     letterSpacing=0
     monospacedWidth=1
     size=24
     skewAmount=0
     typefaceFontName=NotoSans-Regular
     outlineSettings {
      bApplyOutlineToDropShadows=false
      bMiteredCorners=false
      bSeparateFillAlpha=false
      outlineMaterial=null
      outlineSize=0
      outlineColor=<rgba 0 0 0 1>
     }
    }
    renderTransform {angle=0, scale=<0.8 0.8>, shear=<0 0>, translation=<-6 -4>}
    shadowOffset=<0 0>
   }
   slot {class=ButtonSlot}
  }
  ControlBusMixField {
   widgetClassPath=&SinglePropertyView
   parent=CanvasPanel_0
   widget=ControlBusMixField
   slot {class=CanvasPanelSlot, properties {layoutData {alignment=<0 0>, anchors {maximum=<0 0>, minimum=<0 0>}, offsets=<ltrb 0 0 380 60>}}}
  }
  ScrollBox_91 {
   widgetClassPath=&ScrollBox
   parent=CanvasPanel_0
   widget=ScrollBox_91
   properties {
    widgetBarStyle {
     thickness=8
     draggedThumbImage {
      bIsDynamicallyLoaded=false
      drawAs=RoundedBox
      imageType=NoImage
      mirroring=NoMirror
      resourceName=None
      resourceObject=null
      tiling=NoTile
      imageSize=<0 0>
      margin=<ltrb 0 0 0 0>
      outlineSettings {
       bUseBrushTransparency=false
       roundingType=FixedRadius
       width=0
       color {colorUseRule=UseColor_Specified, specifiedColor=<rgba 0 0 0 0>}
       cornerRadii=<4 4 4 4>
      }
      tintColor {colorUseRule=UseColor_Specified, specifiedColor=<rgba 0.2159 0.2159 0.2159 1>}
      uVRegion {bIsValid=false, max=<0 0>, min=<0 0>}
     }
     horizontalBackgroundImage {
      bIsDynamicallyLoaded=false
      drawAs=NoDrawType
      imageType=NoImage
      mirroring=NoMirror
      resourceName=None
      resourceObject=null
      tiling=NoTile
      imageSize=<0 0>
      margin=<ltrb 0 0 0 0>
      outlineSettings {
       bUseBrushTransparency=false
       roundingType=HalfHeightRadius
       width=0
       color {colorUseRule=UseColor_Specified, specifiedColor=<rgba 0 0 0 0>}
       cornerRadii=<0 0 0 0>
      }
      tintColor {colorUseRule=UseColor_Specified, specifiedColor=<rgba 1 1 1 1>}
      uVRegion {bIsValid=false, max=<0 0>, min=<0 0>}
     }
     horizontalBottomSlotImage {
      bIsDynamicallyLoaded=false
      drawAs=NoDrawType
      imageType=NoImage
      mirroring=NoMirror
      resourceName=None
      resourceObject=null
      tiling=NoTile
      imageSize=<0 0>
      margin=<ltrb 0 0 0 0>
      outlineSettings {
       bUseBrushTransparency=false
       roundingType=HalfHeightRadius
       width=0
       color {colorUseRule=UseColor_Specified, specifiedColor=<rgba 0 0 0 0>}
       cornerRadii=<0 0 0 0>
      }
      tintColor {colorUseRule=UseColor_Specified, specifiedColor=<rgba 1 1 1 1>}
      uVRegion {bIsValid=false, max=<0 0>, min=<0 0>}
     }
     horizontalTopSlotImage {
      bIsDynamicallyLoaded=false
      drawAs=NoDrawType
      imageType=NoImage
      mirroring=NoMirror
      resourceName=None
      resourceObject=null
      tiling=NoTile
      imageSize=<0 0>
      margin=<ltrb 0 0 0 0>
      outlineSettings {
       bUseBrushTransparency=false
       roundingType=HalfHeightRadius
       width=0
       color {colorUseRule=UseColor_Specified, specifiedColor=<rgba 0 0 0 0>}
       cornerRadii=<0 0 0 0>
      }
      tintColor {colorUseRule=UseColor_Specified, specifiedColor=<rgba 1 1 1 1>}
      uVRegion {bIsValid=false, max=<0 0>, min=<0 0>}
     }
     hoveredThumbImage {
      bIsDynamicallyLoaded=false
      drawAs=RoundedBox
      imageType=NoImage
      mirroring=NoMirror
      resourceName=None
      resourceObject=null
      tiling=NoTile
      imageSize=<0 0>
      margin=<ltrb 0 0 0 0>
      outlineSettings {
       bUseBrushTransparency=false
       roundingType=FixedRadius
       width=0
       color {colorUseRule=UseColor_Specified, specifiedColor=<rgba 0 0 0 0>}
       cornerRadii=<4 4 4 4>
      }
      tintColor {colorUseRule=UseColor_Specified, specifiedColor=<rgba 0.2159 0.2159 0.2159 1>}
      uVRegion {bIsValid=false, max=<0 0>, min=<0 0>}
     }
     normalThumbImage {
      bIsDynamicallyLoaded=false
      drawAs=RoundedBox
      imageType=NoImage
      mirroring=NoMirror
      resourceName=None
      resourceObject=null
      tiling=NoTile
      imageSize=<0 0>
      margin=<ltrb 0 0 0 0>
      outlineSettings {
       bUseBrushTransparency=false
       roundingType=FixedRadius
       width=0
       color {colorUseRule=UseColor_Specified, specifiedColor=<rgba 0 0 0 0>}
       cornerRadii=<4 4 4 4>
      }
      tintColor {colorUseRule=UseColor_Specified, specifiedColor=<rgba 0.0953 0.0953 0.0953 1>}
      uVRegion {bIsValid=false, max=<0 0>, min=<0 0>}
     }
     verticalBackgroundImage {
      bIsDynamicallyLoaded=false
      drawAs=NoDrawType
      imageType=NoImage
      mirroring=NoMirror
      resourceName=None
      resourceObject=null
      tiling=NoTile
      imageSize=<0 0>
      margin=<ltrb 0 0 0 0>
      outlineSettings {
       bUseBrushTransparency=false
       roundingType=HalfHeightRadius
       width=0
       color {colorUseRule=UseColor_Specified, specifiedColor=<rgba 0 0 0 0>}
       cornerRadii=<0 0 0 0>
      }
      tintColor {colorUseRule=UseColor_Specified, specifiedColor=<rgba 1 1 1 1>}
      uVRegion {bIsValid=false, max=<0 0>, min=<0 0>}
     }
     verticalBottomSlotImage {
      bIsDynamicallyLoaded=false
      drawAs=NoDrawType
      imageType=NoImage
      mirroring=NoMirror
      resourceName=None
      resourceObject=null
      tiling=NoTile
      imageSize=<0 0>
      margin=<ltrb 0 0 0 0>
      outlineSettings {
       bUseBrushTransparency=false
       roundingType=HalfHeightRadius
       width=0
       color {colorUseRule=UseColor_Specified, specifiedColor=<rgba 0 0 0 0>}
       cornerRadii=<0 0 0 0>
      }
      tintColor {colorUseRule=UseColor_Specified, specifiedColor=<rgba 1 1 1 1>}
      uVRegion {bIsValid=false, max=<0 0>, min=<0 0>}
     }
     verticalTopSlotImage {
      bIsDynamicallyLoaded=false
      drawAs=NoDrawType
      imageType=NoImage
      mirroring=NoMirror
      resourceName=None
      resourceObject=null
      tiling=NoTile
      imageSize=<0 0>
      margin=<ltrb 0 0 0 0>
      outlineSettings {
       bUseBrushTransparency=false
       roundingType=HalfHeightRadius
       width=0
       color {colorUseRule=UseColor_Specified, specifiedColor=<rgba 0 0 0 0>}
       cornerRadii=<0 0 0 0>
      }
      tintColor {colorUseRule=UseColor_Specified, specifiedColor=<rgba 1 1 1 1>}
      uVRegion {bIsValid=false, max=<0 0>, min=<0 0>}
     }
    }
    widgetStyle {
     barThickness=8
     bottomShadowBrush {
      bIsDynamicallyLoaded=false
      drawAs=Box
      imageType=FullColor
      mirroring=NoMirror
      resourceName=../../../Engine/Content/Slate/Common/ScrollBoxShadowBottom.png
      resourceObject=null
      tiling=NoTile
      imageSize=<16 8>
      margin=<ltrb 0.5 0 0.5 1>
      outlineSettings {
       bUseBrushTransparency=false
       roundingType=HalfHeightRadius
       width=0
       color {colorUseRule=UseColor_Specified, specifiedColor=<rgba 0 0 0 0>}
       cornerRadii=<0 0 0 0>
      }
      tintColor {colorUseRule=UseColor_Specified, specifiedColor=<rgba 1 1 1 1>}
      uVRegion {bIsValid=false, max=<0 0>, min=<0 0>}
     }
     horizontalScrolledContentPadding=<ltrb 0 0 1 0>
     leftShadowBrush {
      bIsDynamicallyLoaded=false
      drawAs=Box
      imageType=FullColor
      mirroring=NoMirror
      resourceName=../../../Engine/Content/Slate/Common/ScrollBoxShadowLeft.png
      resourceObject=null
      tiling=NoTile
      imageSize=<8 16>
      margin=<ltrb 1 0.5 0 0.5>
      outlineSettings {
       bUseBrushTransparency=false
       roundingType=HalfHeightRadius
       width=0
       color {colorUseRule=UseColor_Specified, specifiedColor=<rgba 0 0 0 0>}
       cornerRadii=<0 0 0 0>
      }
      tintColor {colorUseRule=UseColor_Specified, specifiedColor=<rgba 1 1 1 1>}
      uVRegion {bIsValid=false, max=<0 0>, min=<0 0>}
     }
     rightShadowBrush {
      bIsDynamicallyLoaded=false
      drawAs=Box
      imageType=FullColor
      mirroring=NoMirror
      resourceName=../../../Engine/Content/Slate/Common/ScrollBoxShadowRight.png
      resourceObject=null
      tiling=NoTile
      imageSize=<8 16>
      margin=<ltrb 0 0.5 1 0.5>
      outlineSettings {
       bUseBrushTransparency=false
       roundingType=HalfHeightRadius
       width=0
       color {colorUseRule=UseColor_Specified, specifiedColor=<rgba 0 0 0 0>}
       cornerRadii=<0 0 0 0>
      }
      tintColor {colorUseRule=UseColor_Specified, specifiedColor=<rgba 1 1 1 1>}
      uVRegion {bIsValid=false, max=<0 0>, min=<0 0>}
     }
     topShadowBrush {
      bIsDynamicallyLoaded=false
      drawAs=Box
      imageType=FullColor
      mirroring=NoMirror
      resourceName=../../../Engine/Content/Slate/Common/ScrollBoxShadowTop.png
      resourceObject=null
      tiling=NoTile
      imageSize=<16 8>
      margin=<ltrb 0.5 1 0.5 0>
      outlineSettings {
       bUseBrushTransparency=false
       roundingType=HalfHeightRadius
       width=0
       color {colorUseRule=UseColor_Specified, specifiedColor=<rgba 0 0 0 0>}
       cornerRadii=<0 0 0 0>
      }
      tintColor {colorUseRule=UseColor_Specified, specifiedColor=<rgba 1 1 1 1>}
      uVRegion {bIsValid=false, max=<0 0>, min=<0 0>}
     }
     verticalScrolledContentPadding=<ltrb 0 0 0 1>
    }
   }
   slot {
    class=CanvasPanelSlot
    properties {layoutData {alignment=<0 0>, anchors {maximum=<0 0>, minimum=<0 0>}, offsets=<ltrb 4 148 765.4655 694.2642>}}
   }
  }
  ControlBusses {
   widgetClassPath=&DynamicEntryBox
   bIsVariable=true
   parent=ScrollBox_91
   widget=ControlBusses
   properties {
    clipping=ClipToBounds
    entryBoxType=Wrap
    entryWidgetClass=&/AudioModulationBus_Widget.AudioModulationBus_Widget_C
    numDesignerPreviewEntries=9
    entrySpacing=<22.3454 0>
   }
   slot {class=ScrollBoxSlot}
  }
 ]
}
